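# Niagara ribbon trail effect for the SlugRipple pickaxe melee weapon.
# A single Trail emitter spawns and updates ribbon particles through the shared FN anim-trail modules, with width mode FromSecond.
# Particle color is a dim, semi-transparent warm tint that varies by time of day.
# Emitter.RibbonLife sets particle lifetime and normalizes emitter age for a fade-out curve, which is multiplied by a quick fade-in curve to scale alpha.
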NiagaraSystem /BRCosmetics/Weapons/FORT_Melee/Pickaxe_SlugRipple/FX/NS_Pickaxe_SlugRipple_Trail
dependencies {
 usedDataInterfaces [&NiagaraDataInterfaceCurve, &NiagaraDataInterfaceSocketReader]
 usedDynamicInputs []
 usedModules [
  &self:SystemSpawnScript.SystemScriptSource.SystemScriptGraph.NiagaraNodeAssignment_1.SetVariables_#1
  &SystemState
 ]
}
emitters {Trail {
  emitterSpawnScript [
   &FN_RibbonIDOffsetUtility()
   /SetVariables(Emitter.RibbonLife=0.26) {isSetParameters=true}
  ]
  emitterUpdateScript [
   &EmitterState()
   &FN_SpawnAnimTrail(AnimTrailWidthMode=::FromSecond)
  ]
  particleSpawnScript [
   &FN_InitializeAnimTrailBoneIndecies()
   &FN_InitializeAnimTrail(
    Color=&FNTimeOfDayColor(
     Night Color=<rgba 0.0435 0.0258 0.0369 0.55>
     Morning Color=<rgba 0.08 0.0341 0.0198 0.6>
     Day Color=<rgba 0.115 0.0491 0.0366 0.47>
     Evening Color=<rgba 0.06 0.0205 0.0182 0.55>
    )
    Lifetime=$Emitter.RibbonLife
   )
  ]
  particleUpdateScript [
   &SolveForcesAndVelocity(Clamp Velocity=false, Limit Acceleration=false)
   &FN_UpdateAnimTrail() {module=FN_UpdateAnimTrail001}
   &ScaleColor(
    Scale Alpha=&Multiply_Float(
     A=&FloatFromCurve(
      FloatCurve=DataInterface() {propertyValues {Curve {keys [<at 0.3 1>,<at 0.6317 0.6765>,<at 0.975 0>]}}}
      CurveIndex=&Divide_Float(A=$Emitter.Age, B=$Emitter.RibbonLife)
     )
     B=&FloatFromCurve(FloatCurve=DataInterface() {propertyValues {Curve {keys [{arriveTangent=15.4629, interpMode=RCIM_Cubic, leaveTangent=15.4629, tangentMode=RCTM_User, time=0, value=0}, <at 0.0471 1 RCIM_Cubic>]}}})
    )
    ScaleA=true
    ScaleRGB=true
   ) {module=ScaleColor005}
  ]
  rendererClasses [&NiagaraRibbonRendererProperties]
  renderers [{rendererClass=&NiagaraRibbonRendererProperties}]
 }}
summary {
 systemName=NS_Pickaxe_SlugRipple_Trail
 emitters [{emitterName=Trail, rendererClasses [&NiagaraRibbonRendererProperties]}]
 userVariables [
  User.bIsFrontEnd {type=NiagaraBool}
  User.bIsFrontEndPreview {type=NiagaraBool}
  User.RibbonIDOffset {type=NiagaraInt32}
 ]
}
